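# M_TarpTest is a two-sided Default Lit test material in a developer map folder, named for a tarp.
# It uses only constants, including a dark green color and a scalar, with no textures or parameters.
# The engine's SimpleGrassWind function, meant for world position offset, supplies wind sway from fixed intensity, weight and speed values, repurposing grass motion for the tarp.
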Material /Game/Athena/Asteria/Maps/Test/Dev/Chris/Materials/M_TarpTest
graph {
 properties {pixelDepthOffsetMode=PDOM_Legacy, twoSided=true, usedShadingModels=Default Lit}
 nodes [
  Constant3Vector() {constant=<rgba 0.0149 0.026 0.0033 1>}
  Constant() {r=0.7}
  MaterialFunctionCall!(WindIntensity=Constant() {r=0.2}, WindWeight=Constant() {r=1.5}, WindSpeed=Constant() {r=0.3}, AdditionalWPO=Constant()) {materialFunction=/Engine/Functions/Engine_MaterialFunctions01/WorldPositionOffset/SimpleGrassWind}
 ]
}
