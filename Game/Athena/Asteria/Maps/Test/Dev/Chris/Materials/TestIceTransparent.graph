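# Translucent Default Lit test material for transparent ice, set to index-of-refraction refraction, kept in a developer test map folder.
# A world-aligned, triplanar sample of a shattered-glass normal map from a weapon wrap drives a lerp between 1 and a slightly higher value, a range typical of refraction indices.
# A dimmed blue constant color and a standalone scalar constant make up the rest of the graph.
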Material /Game/Athena/Asteria/Maps/Test/Dev/Chris/Materials/TestIceTransparent
graph {
 properties {blendMode=BLEND_Translucent, pixelDepthOffsetMode=PDOM_Legacy, refractionMethod=RM_IndexOfRefraction, usedShadingModels=Default Lit}
 nodes [
  Constant() {r=0.7}
  LinearInterpolate(Alpha=MaterialFunctionCall(
    TextureObject=TextureObject() {samplerType=SAMPLERTYPE_Normal, texture=/Game/Weapons/WeaponSkins/Wraps/Textures/ShatterFly/ShatterFly_Glass_Normal}
    TextureSize=Constant3Vector() {constant=<rgba 256 256 256 0>}
   ) {materialFunction=&WorldAlignedTexture}.XYZ Texture) {constA=1, constB=1.21}
  Multiply(A=Constant3Vector() {constant=<rgba 0 0.197 0.5312 0>}) {constB=0.2}
 ]
}
